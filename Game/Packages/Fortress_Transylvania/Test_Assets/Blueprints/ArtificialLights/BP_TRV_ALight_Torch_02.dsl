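; BP_TRV_ALight_Torch_02 is a Transylvania fortress torch artificial-light Blueprint, derived from a parent light Blueprint.
; Its construction script runs the parent's setup, then destroys the BloomPlane component when the Turn Off flag is set.
; An EventCallAnimationCurveAgain event is declared but carries no logic in this Blueprint.
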(event EventCallAnimationCurveAgain)

(fn ConstructionScript
  (|CallParentFunction)
  (if (Variables|Default|GetTurn%20Off)
    (Components|DestroyComponent (Variables|Default|GetBloomPlane))))
