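; Torch artificial-light Blueprint for the Transylvania fortress package, built on a parent class with a flame mesh, a particle system and a bloom plane.
; The construction script runs the parent's logic, feeds LightBrightness into the torch particle parameter, and destroys the bloom plane when Turn Off is set.
; GetOverrideMeshMaterial returns the flame's first material when queried for that mesh.
; The EventCallAnimationCurveAgain event has no logic attached.
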(event EventCallAnimationCurveAgain)

(fn GetOverrideMeshMaterial
  (params (MeshComp :type Static%20Mesh%20Component%20Object%20Reference) (MatIdx :type Integer))
  (bind _flame (Variables|Default|GetFlame))
  (if (Utilities|Equal(Object) MeshComp _flame)
    (return (Rendering|Material|GetMaterial _flame 0))))

(fn ConstructionScript
  (|CallParentFunction)
  (Effects|Components|ParticleSystem|SetFloatParameter (Variables|Default|GetP_Athena_TT_Torch) "LightBrightness" (Variables|Default|GetLightBrightness))
  (if (Variables|Default|GetTurn%20Off)
    (Components|DestroyComponent (Variables|Default|GetBloomPlane))))
